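# Niagara system for the idle effect of the SherWolf glider, layering head and hind smoke, mesh trails, speed lines and a spiral trail.
# Smoke and trails follow skeletal mesh samples; trails use per-partition arrays for placement and pass offset positions to materials, plus time-smoothed copies.
# A front-end preview flag switches offsets, timing and colors, while platform pickers swap or scale layers.
# Time-of-day colors tint layers, and a system Timeline modulates several.
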NiagaraSystem /BRCosmetics/Gadgets/Assets/VinderTech_GliderChute/Glider_SherWolf/FX/NS_Glider_SherWolf_Idle
dependencies {
 usedDataInterfaces [
  &NiagaraDataInterfaceArrayFloat3
  &NiagaraDataInterfacePlatformSet
  &NiagaraDataInterfaceCurve
  &NiagaraDataInterfaceColorCurve
  &NiagaraDataInterfaceArrayFloat
  &NiagaraDataInterfaceMeshRendererInfo
  &NiagaraDataInterfaceSkeletalMesh
 ]
 usedDynamicInputs [&SetBoolByFloatComparison]
 usedModules [
  &self:SystemSpawnScript.SystemScriptSource.SystemScriptGraph.NiagaraNodeAssignment_3.SetVariables_#1
  &self:SystemSpawnScript.SystemScriptSource.SystemScriptGraph.NiagaraNodeAssignment_4.SetVariables_#2
  &self:SystemSpawnScript.SystemScriptSource.SystemScriptGraph.NiagaraNodeAssignment_0.SetVariables_#3
  &self:SystemSpawnScript.SystemScriptSource.SystemScriptGraph.NiagaraNodeAssignment_1.SetVariables_#4
  &SystemState
  &Timeline
 ]
}
emitters {
 Head_Smoke {
  emitterUpdateScript [
   &EmitterState(
    Delay First Loop Only=true
    Inactive Response=::NewEnumerator1
    Life Cycle Mode=::NewEnumerator1
    Loop Delay=&MakeCustomFloatFromBool(Bool=$User.bIsFrontEndPreview, True Float=2, False Float=1.8)
    Loop Duration=1
    UseLoopDelay=true
   )
   &SpawnRate(
    SpawnRate=&MakeCustomFloatFromBool(
     Bool=$User.bIsFrontEndPreview
     True Float=0
     False Float=&FN_PickFloatDependOnPlatform(TargetPlatformActive=3, TargetPlatformNoActive=0)
    )
    Use Spawn Probability=false
   )
  ]
  particleSpawnScript [
   &PartitionParticles()
   &InitializeParticle(
    Color=&MakeCustomLinearColorFromBool(Bool=$User.bIsFrontEndPreview, True Color=<rgba 0.6129 0.7074 0.8229 1>, False Color=<rgba 1 1 1 1>)
    Lifetime=1.5
    Mesh Scale=<1.3 1.3 1.3>
    Mesh Scale Mode=::NewEnumerator4
    UsePositionOffset=true
    Write Material Random=true
   )
   &InitialMeshOrientation(
    Mesh Orientation Mode=::NewEnumerator2
    Rotation=&RandomRangeVector(Minimum=<0 0 0>, Maximum=<1 0 0>)
    Use Rotation Vector=true
   ) {module=InitialMeshOrientation001}
  ]
  particleUpdateScript [
   &ParticleState()
   &SkeletalMeshLocation(
    Bone Sampling Mode=::NewEnumerator4
    MaskSpawnGroup=false
    Sampled Position Offset=&MakeCustomVectorFromBool(
     Bool=$User.bIsFrontEndPreview
     True Vector=&MakeVector(
      X=&FloatFromCurve(FloatCurve=DataInterface() {propertyValues {Curve {keys [<at 0 -50>,<at 1 -80>]}}})
      Y=&FloatFromCurve(
       FloatCurve=DataInterface() {propertyValues {Curve {keys [<at 0 -80>,<at 1 37.5>]}}}
       Scale Curve=&Multiply_Float(
        A=&FN_FrontEndFloatModifier(Gameplay=1.5, Front End=1)
        B=&FN_PickFloatDependOnPlatform(TargetPlatformActive=1, TargetPlatformNoActive=0.9)
       )
      )
      Z=&FloatFromCurve(FloatCurve=DataInterface() {propertyValues {Curve {keys [
           {arriveTangent=-3.3453, interpMode=RCIM_Cubic, leaveTangent=-3.3453, tangentMode=RCTM_User, time=0, value=35}
           {arriveTangent=-1.1949, interpMode=RCIM_Cubic, leaveTangent=-1.1949, tangentMode=RCTM_User, time=1, value=25}
          ]}}})
     )
     False Vector=&MakeVector(
      X=&RandomRangeFloat(Minimum=-5, Maximum=5)
      Y=&FloatFromCurve(
       FloatCurve=DataInterface() {propertyValues {Curve {keys [<at 0 80>,<at 1 -50>]}}}
       Scale Curve=&Multiply_Float(
        A=&FN_FrontEndFloatModifier(Gameplay=1.5, Front End=1)
        B=&FN_PickFloatDependOnPlatform(TargetPlatformActive=1, TargetPlatformNoActive=0.9)
       )
      )
      Z=&FloatFromCurve(FloatCurve=DataInterface() {propertyValues {Curve {keys [
           {arriveTangent=-3.3453, interpMode=RCIM_Cubic, leaveTangent=-3.3453, tangentMode=RCTM_User, time=0, value=22.5}
           {arriveTangent=-40, interpMode=RCIM_Cubic, leaveTangent=-40, time=0.1215, value=21.0887}
           {arriveTangent=-27.1436, interpMode=RCIM_Cubic, leaveTangent=-27.1436, time=0.5, value=2.5}
           {arriveTangent=-1.1949, interpMode=RCIM_Cubic, leaveTangent=-1.1949, tangentMode=RCTM_User, time=1, value=-2.7561}
          ]}}})
     )
    )
    Skeletal Mesh=$System.Skeletal Mesh
   )
   &ScaleMeshSize(Scale Factor=&VectorFromFloat(Value=&FloatFromCurve(
      FloatCurve=DataInterface() {propertyValues {Curve {keys [<at 0 0.7>,<at 0.2 1>,<at 0.5 0.9>,<at 1 0.2>]}}}
      Scale Curve=&MakeCustomFloatFromBool(
       Bool=$User.bIsFrontEndPreview
       True Float=&RandomRangeFloat(Minimum=0.6, Maximum=0.7)
       False Float=&RandomRangeFloat(Minimum=0.9, Maximum=1)
      )
     )))
   &CameraOffset(Camera Offset Amount=&FN_FrontEndFloatModifier(
     Gameplay=&FN_PickFloatDependOnPlatform(TargetPlatformActive=-5, TargetPlatformNoActive=-10)
     Front End=-20
    ))
   &ScaleColor(
    Scale Alpha=0.8
    Scale RGB=&MakeVectorFromLinearColorRGB(Color=$System.TODColor)
    ScaleA=true
    ScaleRGB=true
   ) {module=ScaleColor001}
   &ScaleColor(
    Scale Alpha=&FloatFromCurve(FloatCurve=DataInterface() {propertyValues {Curve {keys [
         {arriveTangent=5.611, interpMode=RCIM_Cubic, leaveTangent=5.611, tangentMode=RCTM_User, time=0, value=0}
         {arriveTangent=0.0117, interpMode=RCIM_Cubic, leaveTangent=0.0117, tangentMode=RCTM_User, time=0.2, value=1}
         <at 1 0 RCIM_Cubic>
        ]}}})
    Scale RGB=&VectorFromFloat(Value=&Multiply_Float(A=&RandomRangeFloat(Minimum=0.8, Maximum=1), B=1))
    ScaleA=true
    ScaleRGB=true
   )
   &DynamicMaterialParameters(
    Index 0 Param 1=&FloatFromCurve(
     FloatCurve=DataInterface() {propertyValues {Curve {keys [
         {interpMode=RCIM_Cubic, tangentMode=RCTM_User, time=0, value=0}
         <at 0.2 0.55>
         {arriveTangent=-1.0223, interpMode=RCIM_Cubic, leaveTangent=-1.0223, tangentMode=RCTM_User, time=1, value=1}
        ]}}}
     Scale Curve=&RandomRangeFloat(Minimum=2, Maximum=5)
    )
    Index 0 Param 3=1
    Index 0 Param 4=&FloatFromCurve(FloatCurve=DataInterface() {propertyValues {Curve {keys [{arriveTangent=-2.87, interpMode=RCIM_Cubic, leaveTangent=-2.87, tangentMode=RCTM_User, time=0.3, value=1}, <at 1 0 RCIM_Cubic>]}}})
    Module.Index0Param0Write=true
    Module.Index0Param1Write=true
    Module.Index0Param2Write=true
    Module.Index0Param3Write=true
   )
  ]
  rendererClasses [&NiagaraMeshRendererProperties]
  renderers [{rendererClass=&NiagaraMeshRendererProperties, rendererIndex=0}]
 }
 Head_Trail {
  emitterUpdateScript [
   &EmitterState(
    Inactive Response=::NewEnumerator1
    Life Cycle Mode=::NewEnumerator1
    Loop Behavior=::NewEnumerator1
    Loop Delay=&MakeCustomFloatFromBool(Bool=$User.bIsFrontEndPreview, True Float=2.2, False Float=2)
    Loop Duration Mode=::NewEnumerator1
    UseLoopDelay=true
   )
   &SpawnBurst_Instantaneous(
    Spawn Count=&MakeCustomIntFromBool(Bool=$User.bIsFrontEndPreview, True INT=6, False INT=8)
    Use Spawn Probability=false
    UseLoopCountLimit=false
   )
   /SetVariables(Emitter.PositionFix:NiagaraFloat=&FN_FrontEndFloatModifier(Gameplay=1, Front End=1)) {isSetParameters=true}
  ]
  particleSpawnScript [
   &PartitionParticles(Number Of Partitions=&MakeCustomIntFromBool(Bool=$User.bIsFrontEndPreview, True INT=6, False INT=8))
   &InitializeParticle(
    Color Maximum=<rgba 0.015 0.015 0.015 1.2>
    Color Minimum=<rgba 0.03 0.03 0.03 1.2>
    Color Mode=::NewEnumerator2
    Mesh Scale=&MakeCustomVectorFromBool(
     Bool=$User.bIsFrontEndPreview
     True Vector=&SelectVectorFromArray(
      Direct Array Index=$Particles.PartitionParticles.Partition
      Vector Selection Array=DataInterface() {propertyValues {FloatData [<1 0.6 1>, <0.8 0.75 0.8>, <1 0.75 1>, <-1 0.6 1>, <-0.85 0.75 0.85>, <-1 0.8 1>, <1 0.65 1>, <-1 0.7 1>]}}
     )
     False Vector=&SelectVectorFromArray(
      Direct Array Index=$Particles.PartitionParticles.Partition
      Vector Selection Array=DataInterface() {propertyValues {FloatData [<1 0.7 1>, <0.8 0.75 0.8>, <1 0.75 1>, <-1 0.68 1>, <-0.85 0.75 0.85>, <-1 0.8 1>, <1 0.65 1>, <-1 0.7 1>]}}
     )
    )
    Mesh Scale Mode=::NewEnumerator4
    UsePositionOffset=true
    Write Material Random=true
   )
   &InitialMeshOrientation(
    Mesh Orientation Mode=::NewEnumerator2
    Rotation=&MakeCustomVectorFromBool(Bool=$User.bIsFrontEndPreview, True Vector=<0 0 1>, False Vector=<0.5 0 0>)
    Use Rotation Vector=true
   )
   &InitialMeshOrientation(
    Mesh Orientation Mode=::NewEnumerator2
    Rotation=&SelectVectorFromArray(
     Direct Array Index=$Particles.PartitionParticles.Partition
     Vector Selection Array=DataInterface() {propertyValues {FloatData [<0 0.09 0>, <0 0.11 0>, <0 0.15 0>, <0 0.1 0>, <0 -0.11 0>, <0 0.1 0>, <0 0 0>, <0 0.11 0>]}}
    )
    Use Rotation Vector=true
   ) {module=InitialMeshOrientation001}
   &InitialMeshOrientation(
    Mesh Orientation Mode=::NewEnumerator2
    Rotation=&MakeCustomVectorFromBool(
     Bool=$User.bIsFrontEndPreview
     True Vector=&SelectVectorFromArray(
      Direct Array Index=$Particles.PartitionParticles.Partition
      Vector Selection Array=DataInterface() {propertyValues {FloatData [<0.02 0 0.05>, <0.015 0 0.06>, <0.01 0 0.04>, <0.03 0 -0.02>, <0.02 0 -0.03>, <0.02 0 -0.02>, <0.02 0 0.05>, <0.04 0 -0.04>]}}
     )
     False Vector=&SelectVectorFromArray(
      Direct Array Index=$Particles.PartitionParticles.Partition
      Vector Selection Array=DataInterface() {propertyValues {FloatData [<0.02 0 0.03>, <0.01 0 0.05>, <0 0 0.02>, <0.03 0 -0.08>, <0.02 0 -0.06>, <0 0 -0.02>, <0.1 0 0.05>, <0.04 0 -0.05>]}}
     )
    )
    Use Rotation Vector=true
   ) {module=InitialMeshOrientation002}
   &CameraOffset(Camera Offset Amount=10)
  ]
  particleUpdateScript [
   &ParticleState(Kill Particles When Lifetime Has Elapsed=false, Let Infinitely Lived Particles Die When Emitter Deactivates=true)
   &SkeletalMeshLocation(
    Bone Sampling Mode=::NewEnumerator4
    MaskSpawnGroup=false
    Sampled Position Offset=&AddVectorToPosition(
     Position=&MakeCustomVectorFromBool(
      Bool=$User.bIsFrontEndPreview
      True Vector=&SelectVectorFromArray(
       Direct Array Index=$Particles.PartitionParticles.Partition
       Vector Selection Array=DataInterface() {propertyValues {FloatData [<-70 -100 30>, <-80 -90 10>, <-90 -70 -15>, <-50 -100 25>, <-40 -90 12>, <-40 -80 -15>, <-80 30 -10>, <-70 20 -10>]}}
      )
      False Vector=&SelectVectorFromArray(
       Direct Array Index=$Particles.PartitionParticles.Partition
       Vector Selection Array=DataInterface() {propertyValues {FloatData [<10 130 30>, <30 120 -10>, <35 100 -20>, <-5 130 30>, <-20 115 10>, <-35 100 -20>, <10 -20 -20>, <-10 -10 -35>]}}
      )
     )
     Vector=&Lerp_Vector(
      A=<0 0 10>
      B=<0 0 10>
      Alpha=&FN_PickFloatDependOnPlatform(TargetPlatformActive=1, TargetPlatformNoActive=0)
     )
    )
    Skeletal Mesh=$System.Skeletal Mesh
   )
   &DynamicMaterialParameters(
    Index 0 Param 1=&RandomRangeFloat(Minimum=0.2, Maximum=1)
    Index 0 Param 2=&RandomRangeFloat(Minimum=0.2, Maximum=1)
    Index 0 Param 3=1
    Index 0 Param 4=20
    Index 2 Param 1=&MakeFloatFromVector(
     Vector=&AddVectorToPosition(
      Position=$Particles.Position
      Vector=&TransformVector(
       Vector=&FN_FrontEndVector3Modifier(Gameplay=$System.IGPOFix, Front End=$System.FEPoFix)
       Destination Space=::Simulation
      )
     )
     Channel=::NewEnumerator0
    )
    Index 2 Param 2=&MakeFloatFromVector(
     Vector=&AddVectorToPosition(
      Position=$Particles.Position
      Vector=&TransformVector(
       Vector=&FN_FrontEndVector3Modifier(Gameplay=$System.IGPOFix, Front End=$System.FEPoFix)
       Destination Space=::Simulation
      )
     )
     Channel=::NewEnumerator1
    )
    Index 2 Param 3=&MakeFloatFromVector(
     Vector=&AddVectorToPosition(
      Position=$Particles.Position
      Vector=&TransformVector(
       Vector=&FN_FrontEndVector3Modifier(Gameplay=$System.IGPOFix, Front End=$System.FEPoFix)
       Destination Space=::Simulation
      )
     )
     Channel=::NewEnumerator2
    )
    Index 3 Param 1=&MakeCustomFloatFromBool(
     Bool=$User.bIsFrontEndPreview
     True Float=&FloatFromCurve(
      FloatCurve=DataInterface() {propertyValues {Curve {keys [{arriveTangent=237.3113, time=0, value=-800}, {arriveTangent=816.2405, leaveTangent=816.2405, time=1, value=0}]}}}
      Scale Curve=&RandomRangeFloat(Minimum=0.8, Maximum=1)
     )
     False Float=&FloatFromCurve(
      FloatCurve=DataInterface() {propertyValues {Curve {keys [{arriveTangent=237.3113, time=0, value=-600}, {arriveTangent=816.2405, leaveTangent=816.2405, time=1, value=0}]}}}
      Scale Curve=&RandomRangeFloat(Minimum=0.8, Maximum=1)
     )
    )
    Module.Index0Param0Write=true
    Module.Index0Param1Write=true
    Module.Index0Param2Write=true
    Module.Index0Param3Write=true
    Module.Index2Param0Write=true
    Module.Index2Param1Write=true
    Module.Index2Param2Write=true
    Module.Index2Param3Write=true
    Module.Index3Param0Write=true
    Module.Index3Param1Write=true
    Module.Index3Param2Write=true
    Module.Index3Param3Write=true
    Write Parameter Index 2=true
    Write Parameter Index 3=true
   )
   &InterpolateOverTime(
    Rate Of Change:NiagaraFloat=&FN_FrontEndFloatModifier(Gameplay=10, Front End=10)
    Target Vector Value:Vector3f=&AddVectorToPosition(
     Position=$Particles.Position
     Vector=&TransformVector(
      Vector=&FN_FrontEndVector3Modifier(Gameplay=$System.IGPOFix, Front End=$System.FEPoFix)
      Destination Space=::Simulation
     )
    )
    Variable Type=::NewEnumerator2
   )
   /SetVariables(Particles.Interpolate Dynamic=&MakeFloat4FromVectorAndFloat(
     "Vector (RGB)"=$StackContext.InterpolateOverTime.InterpolateOverTime_V3_.Moving Average
     "Float (Alpha)"=&Multiply_Float(
      A=&FN_FrontEndFloatModifier(Gameplay=0.8, Front End=3)
      B=&RandomRangeFloat(Minimum=0.8, Maximum=1)
     )
    )) {isSetParameters=true}
   /SetVariables(Particles.Color002=&MakeCustomLinearColorFromBool(
     Bool=$User.bIsFrontEndPreview
     True Color=<rgba 0.0225 0.0244 0.03 0.5>
     False Color=&MakeLinearColorFromVectorAndFloat("Vector (RGB)"=&MakeVectorFromLinearColorRGB(Color=$System.TODColor), "Float (Alpha)"=0.01)
    )) {isSetParameters=true}
   &ScaleColor(
    Scale Alpha=&FloatFromCurve(
     FloatCurve=DataInterface() {propertyValues {Curve {keys [<at 0 0>,<at 0.25 1>]}}}
     Scale Curve=&FN_FrontEndFloatModifier(
      Gameplay=&Multiply_Float(
       A=&FN_PickFloatDependOnPlatform(TargetPlatformActive=0.8, TargetPlatformNoActive=0.8)
       B=&FNTimeOfDayScalar(Night Scale=0.7, Morning Scale=1, Evening Scale=1)
      )
      Front End=0.5
     )
    )
    Scale RGB=&MakeVectorFromLinearColorRGB(Color=&MakeCustomLinearColorFromBool(Bool=$User.bIsFrontEndPreview, True Color=<rgba 0.06 0.123 0.2 1>, False Color=$System.TODColor))
    ScaleA=true
    ScaleRGB=true
   )
   &ScaleMeshSize(Scale Factor=&FN_FrontEndVector3Modifier(
     Gameplay=&Multiply_Vector(
      A=<1 1 1>
      B=&VectorFromFloat(Value=&Lerp_Float(A=1, B=1, Alpha=$System.Timeline.Time))
     )
     Front End=<1 1 1>
    ))
  ]
  rendererClasses [&NiagaraMeshRendererProperties]
  renderers [
   {rendererClass=&NiagaraMeshRendererProperties, rendererIndex=0}
   {rendererClass=&NiagaraMeshRendererProperties, rendererIndex=1}
  ]
 }
 Hind_Smoke {
  emitterUpdateScript [
   &EmitterState(
    Delay First Loop Only=true
    Inactive Response=::NewEnumerator1
    Life Cycle Mode=::NewEnumerator1
    Loop Delay=&MakeCustomFloatFromBool(Bool=$User.bIsFrontEndPreview, True Float=2, False Float=0)
    Loop Duration=1
    UseLoopDelay=true
   )
   &SpawnRate(
    Spawn Probability=&FN_PickFloatDependOnPlatform(TargetPlatformActive=1, TargetPlatformNoActive=0)
    SpawnRate=&MakeCustomFloatFromBool(Bool=$User.bIsFrontEndPreview, True Float=8, False Float=10)
    Use Spawn Probability=true
   )
  ]
  particleSpawnScript [
   &InitializeParticle(
    Color Minimum=<rgba 0.5 0.5 0.5 1>
    Color Mode=::NewEnumerator2
    Sprite Rotation Mode=::NewEnumerator2
    Sprite Size Mode=::NewEnumerator1
    Uniform Sprite Size Max=200
    Uniform Sprite Size Min=180
    UsePositionOffset=true
    Write Material Random=true
   )
   &CameraOffset(Camera Offset Amount=&MakeCustomFloatFromBool(
     Bool=$User.bIsFrontEndPreview
     True Float=40
     False Float=&FN_PickFloatDependOnPlatform(TargetPlatformActive=80, TargetPlatformNoActive=40)
    ))
  ]
  particleUpdateScript [
   &ParticleState()
   &SkeletalMeshLocation(
    Bone Sampling Mode=::NewEnumerator4
    MaskSpawnGroup=false
    Sampled Position Offset=&MakeCustomVectorFromBool(
     Bool=$User.bIsFrontEndPreview
     True Vector=&MakeVector(
      X=&FloatFromCurve(FloatCurve=DataInterface() {propertyValues {Curve {keys [<at 0 -60>,<at 1 -80>]}}})
      Y=&FloatFromCurve(
       FloatCurve=DataInterface() {propertyValues {Curve {keys [<at 0 -57.5>,<at 1 120>]}}}
       Scale Curve=&Multiply_Float(
        A=&FN_FrontEndFloatModifier(Gameplay=1.5, Front End=1)
        B=&FN_PickFloatDependOnPlatform(TargetPlatformActive=1, TargetPlatformNoActive=0.9)
       )
      )
      Z=&FloatFromCurve(FloatCurve=DataInterface() {propertyValues {Curve {keys [{arriveTangent=-3.3453, time=0, value=-20}, {arriveTangent=-1.1949, leaveTangent=-1.1949, time=1, value=30}]}}})
     )
     False Vector=&MakeVector(
      X=&RandomRangeFloat(Minimum=-10, Maximum=10)
      Y=&FloatFromCurve(
       FloatCurve=DataInterface() {propertyValues {Curve {keys [<at 0 0>,<at 1 -130>]}}}
       Scale Curve=&Multiply_Float(
        A=&FN_FrontEndFloatModifier(Gameplay=1.5, Front End=1)
        B=&FN_PickFloatDependOnPlatform(TargetPlatformActive=1, TargetPlatformNoActive=0.9)
       )
      )
      Z=&FloatFromCurve(FloatCurve=DataInterface() {propertyValues {Curve {keys [{arriveTangent=-3.3453, time=0, value=-40}, {arriveTangent=-1.1949, leaveTangent=-1.1949, time=1, value=-40}]}}})
     )
    )
    Skeletal Mesh=$System.Skeletal Mesh
   )
   &ScaleColor(Scale RGB=&MakeVectorFromLinearColorRGB(Color=$System.TODColor), ScaleA=true, ScaleRGB=true)
   &ScaleColor(
    Scale Alpha=&Multiply_Float(
     A=&FloatFromCurve(FloatCurve=DataInterface() {propertyValues {Curve {keys [
          <at 0 0 RCIM_Cubic>
          {arriveTangent=-0.0081, interpMode=RCIM_Cubic, leaveTangent=-0.0081, tangentMode=RCTM_User, time=0.3, value=1}
          {arriveTangent=-1.4952, interpMode=RCIM_Cubic, leaveTangent=-1.4952, tangentMode=RCTM_User, time=0.7, value=0.6}
          {arriveTangent=-1.1795, interpMode=RCIM_Cubic, leaveTangent=-1.1795, tangentMode=RCTM_User, time=1, value=0}
         ]}}})
     B=&FloatFromCurve(FloatCurve=DataInterface() {propertyValues {Curve {keys [<at 0 0>,<at 1 0.8>,<at 2.5 1>]}}}, CurveIndex=$Emitter.Age)
    )
    Scale RGB=&MakeCustomVectorFromBool(
     Bool=$User.bIsFrontEndPreview
     True Vector=&Multiply_VectorByFloat(
      Vector=&MakeVectorFromLinearColorRGB(Color=&ColorFromCurve(ColorCurve=DataInterface() {propertyValues {
          AlphaCurve {keys []}
          BlueCurve {keys [<at 0 0.1146>,<at 0.52 0.015 RCIM_Cubic>]}
          GreenCurve {keys [<at 0 0.0716>,<at 0.52 0.0112 RCIM_Cubic>]}
          RedCurve {keys [<at 0 0.0573>,<at 0.52 0.01 RCIM_Cubic>]}
         }}))
      Float=&RandomRangeFloat(Minimum=0.5, Maximum=1)
     )
     False Vector=&Multiply_VectorByFloat(
      Vector=&MakeVectorFromLinearColorRGB(Color=&FNTimeOfDayColor(
        Night Color=&ColorFromCurve(ColorCurve=DataInterface() {propertyValues {
           AlphaCurve {keys []}
           BlueCurve {keys [<at 0 0.1>,<at 0.52 0.01 RCIM_Cubic>]}
           GreenCurve {keys [<at 0 0.1>,<at 0.52 0.01 RCIM_Cubic>]}
           RedCurve {keys [<at 0 0.1>,<at 0.52 0.01 RCIM_Cubic>]}
          }})
        Morning Color=&ColorFromCurve(ColorCurve=DataInterface() {propertyValues {
           AlphaCurve {keys []}
           BlueCurve {keys [<at 0 0.1>,<at 0.52 0.05 RCIM_Cubic>]}
           GreenCurve {keys [<at 0 0.1>,<at 0.52 0.05 RCIM_Cubic>]}
           RedCurve {keys [<at 0 0.1>,<at 0.52 0.05 RCIM_Cubic>]}
          }})
        Day Color=&ColorFromCurve(ColorCurve=DataInterface() {propertyValues {
           AlphaCurve {keys []}
           BlueCurve {keys [<at 0 0.15>,<at 0.52 0.15 RCIM_Cubic>]}
           GreenCurve {keys [<at 0 0.15>,<at 0.52 0.15 RCIM_Cubic>]}
           RedCurve {keys [<at 0 0.15>,<at 0.52 0.15 RCIM_Cubic>]}
          }})
        Evening Color=&ColorFromCurve(ColorCurve=DataInterface() {propertyValues {
           AlphaCurve {keys []}
           BlueCurve {keys [<at 0 0.1>,<at 0.52 0.05 RCIM_Cubic>]}
           GreenCurve {keys [<at 0 0.1>,<at 0.52 0.05 RCIM_Cubic>]}
           RedCurve {keys [<at 0 0.1>,<at 0.52 0.05 RCIM_Cubic>]}
          }})
       ))
      Float=&RandomRangeFloat(Minimum=0.5, Maximum=1)
     )
    )
    ScaleA=true
    ScaleRGB=true
   ) {module=ScaleColor001}
  ]
  rendererClasses [&NiagaraSpriteRendererProperties]
  renderers [{rendererClass=&NiagaraSpriteRendererProperties}]
 }
 Hind_Smoke_Low {
  emitterUpdateScript [
   &EmitterState(
    Delay First Loop Only=true
    Inactive Response=::NewEnumerator1
    Life Cycle Mode=::NewEnumerator1
    Loop Delay=&MakeCustomFloatFromBool(Bool=$User.bIsFrontEndPreview, True Float=2, False Float=0)
    Loop Duration=1
    UseLoopDelay=true
   )
   &SpawnRate(
    Spawn Probability=&FN_PickFloatDependOnPlatform(TargetPlatformActive=0, TargetPlatformNoActive=1)
    SpawnRate=&MakeCustomFloatFromBool(Bool=$User.bIsFrontEndPreview, True Float=8, False Float=8)
    Use Spawn Probability=true
   )
  ]
  particleSpawnScript [
   &InitializeParticle(
    Color Minimum=<rgba 1 1 1 0.5>
    Color Mode=::NewEnumerator2
    Lifetime=1.5
    Sprite Rotation Mode=::NewEnumerator2
    Sprite Size Mode=::NewEnumerator1
    Uniform Sprite Size Max=160
    Uniform Sprite Size Min=140
    UsePositionOffset=true
    Write Material Random=true
   )
   &CameraOffset(Camera Offset Amount=&MakeCustomFloatFromBool(
     Bool=$User.bIsFrontEndPreview
     True Float=40
     False Float=&FN_PickFloatDependOnPlatform(TargetPlatformActive=0, TargetPlatformNoActive=120)
    ))
  ]
  particleUpdateScript [
   &ParticleState()
   &SkeletalMeshLocation(
    Bone Sampling Mode=::NewEnumerator4
    MaskSpawnGroup=false
    Sampled Position Offset=&MakeCustomVectorFromBool(
     Bool=$User.bIsFrontEndPreview
     True Vector=&MakeVector(
      X=&FloatFromCurve(FloatCurve=DataInterface() {propertyValues {Curve {keys [<at 0 -60>,<at 1 -80>]}}})
      Y=&FloatFromCurve(
       FloatCurve=DataInterface() {propertyValues {Curve {keys [<at 0 -57.5>,<at 1 120>]}}}
       Scale Curve=&Multiply_Float(A=&FN_FrontEndFloatModifier(Gameplay=1.5, Front End=1), B=1)
      )
      Z=&FloatFromCurve(FloatCurve=DataInterface() {propertyValues {Curve {keys [{arriveTangent=-3.3453, time=0, value=-20}, {arriveTangent=-1.1949, leaveTangent=-1.1949, time=1, value=30}]}}})
     )
     False Vector=&MakeVector(
      X=&RandomRangeFloat(Minimum=-10, Maximum=10)
      Y=&FloatFromCurve(
       FloatCurve=DataInterface() {propertyValues {Curve {keys [<at 0 120>,<at 1 -100>]}}}
       Scale Curve=&Multiply_Float(A=&FN_FrontEndFloatModifier(Gameplay=1.5, Front End=1), B=1)
      )
      Z=&FloatFromCurve(FloatCurve=DataInterface() {propertyValues {Curve {keys [{arriveTangent=-3.3453, time=0, value=15}, <at 0.4 -12.5>, {arriveTangent=-1.1949, leaveTangent=-1.1949, time=1, value=-30}]}}})
     )
    )
    Skeletal Mesh=$System.Skeletal Mesh
   )
   &ScaleColor(
    Scale Alpha=&Multiply_Float(
     A=&FloatFromCurve(
      FloatCurve=DataInterface() {propertyValues {Curve {keys [
          <at 0 0 RCIM_Cubic>
          {arriveTangent=-0.0081, interpMode=RCIM_Cubic, leaveTangent=-0.0081, tangentMode=RCTM_User, time=0.3, value=1}
          {arriveTangent=-1.4952, interpMode=RCIM_Cubic, leaveTangent=-1.4952, tangentMode=RCTM_User, time=0.7, value=0.6}
          {arriveTangent=-1.1795, interpMode=RCIM_Cubic, leaveTangent=-1.1795, tangentMode=RCTM_User, time=1, value=0}
         ]}}}
      Scale Curve=1.5
     )
     B=&FloatFromCurve(FloatCurve=DataInterface() {propertyValues {Curve {keys [<at 0 0>,<at 1 0.8>,<at 2.5 1>]}}}, CurveIndex=$Emitter.Age)
    )
    Scale RGB=&MakeVectorFromLinearColorRGB(Color=&ColorFromCurve(ColorCurve=DataInterface() {propertyValues {
        AlphaCurve {keys []}
        BlueCurve {keys [<at 0 0.1>,<at 0.52 0.01 RCIM_Cubic>]}
        GreenCurve {keys [<at 0 0.1>,<at 0.52 0.01 RCIM_Cubic>]}
        RedCurve {keys [<at 0 0.1>,<at 0.52 0.01 RCIM_Cubic>]}
       }}))
    ScaleA=true
    ScaleRGB=true
   ) {enabled=false}
   &ScaleColor(
    Scale Alpha=&Multiply_Float(
     A=&FloatFromCurve(
      FloatCurve=DataInterface() {propertyValues {Curve {keys [
          <at 0 0 RCIM_Cubic>
          {arriveTangent=-0.0081, interpMode=RCIM_Cubic, leaveTangent=-0.0081, tangentMode=RCTM_User, time=0.3, value=1}
          {arriveTangent=-1.4952, interpMode=RCIM_Cubic, leaveTangent=-1.4952, tangentMode=RCTM_User, time=0.7, value=0.6}
          {arriveTangent=-1.1795, interpMode=RCIM_Cubic, leaveTangent=-1.1795, tangentMode=RCTM_User, time=1, value=0}
         ]}}}
      Scale Curve=1.5
     )
     B=&FloatFromCurve(FloatCurve=DataInterface() {propertyValues {Curve {keys [<at 0 0>,<at 1 0.8>,<at 2.5 1>]}}}, CurveIndex=$Emitter.Age)
    )
    Scale RGB=&MakeCustomVectorFromBool(
     Bool=$User.bIsFrontEndPreview
     True Vector=&Multiply_VectorByFloat(
      Vector=&MakeVectorFromLinearColorRGB(Color=&ColorFromCurve(ColorCurve=DataInterface() {propertyValues {
          AlphaCurve {keys []}
          BlueCurve {keys [<at 0 0.1>,<at 0.52 0.01 RCIM_Cubic>]}
          GreenCurve {keys [<at 0 0.073>,<at 0.52 0.0063 RCIM_Cubic>]}
          RedCurve {keys [<at 0 0.0641>,<at 0.52 0.005 RCIM_Cubic>]}
         }}))
      Float=&RandomRangeFloat(Minimum=0.5, Maximum=1)
     )
     False Vector=&Multiply_VectorByFloat(
      Vector=&MakeVectorFromLinearColorRGB(Color=&FNTimeOfDayColor(
        Night Color=&ColorFromCurve(ColorCurve=DataInterface() {propertyValues {
           AlphaCurve {keys []}
           BlueCurve {keys [<at 0 0.1>,<at 0.52 0.01 RCIM_Cubic>]}
           GreenCurve {keys [<at 0 0.0595>,<at 0.52 0.01 RCIM_Cubic>]}
           RedCurve {keys [<at 0 0.0279>,<at 0.52 0.01 RCIM_Cubic>]}
          }})
        Morning Color=&ColorFromCurve(ColorCurve=DataInterface() {propertyValues {
           AlphaCurve {keys []}
           BlueCurve {keys [<at 0 0.1>,<at 0.52 0.01 RCIM_Cubic>]}
           GreenCurve {keys [<at 0 0.1>,<at 0.52 0.01 RCIM_Cubic>]}
           RedCurve {keys [<at 0 0.1>,<at 0.52 0.01 RCIM_Cubic>]}
          }})
        Day Color=&ColorFromCurve(ColorCurve=DataInterface() {propertyValues {
           AlphaCurve {keys []}
           BlueCurve {keys [<at 0 0.1>,<at 0.52 0.01 RCIM_Cubic>]}
           GreenCurve {keys [<at 0 0.1>,<at 0.52 0.01 RCIM_Cubic>]}
           RedCurve {keys [<at 0 0.1>,<at 0.52 0.01 RCIM_Cubic>]}
          }})
        Evening Color=&ColorFromCurve(ColorCurve=DataInterface() {propertyValues {
           AlphaCurve {keys []}
           BlueCurve {keys [<at 0 0.1>,<at 0.52 0.01 RCIM_Cubic>]}
           GreenCurve {keys [<at 0 0.1>,<at 0.52 0.01 RCIM_Cubic>]}
           RedCurve {keys [<at 0 0.1>,<at 0.52 0.01 RCIM_Cubic>]}
          }})
       ))
      Float=&RandomRangeFloat(Minimum=0.5, Maximum=1)
     )
    )
    ScaleA=true
    ScaleRGB=true
   ) {module=ScaleColor001}
  ]
  rendererClasses [&NiagaraSpriteRendererProperties]
  renderers [{rendererClass=&NiagaraSpriteRendererProperties}]
 }
 Hind_Trail {
  emitterUpdateScript [
   &EmitterState(
    Inactive Response=::NewEnumerator1
    Life Cycle Mode=::NewEnumerator1
    Loop Behavior=::NewEnumerator1
    Loop Delay=&MakeCustomFloatFromBool(Bool=$User.bIsFrontEndPreview, True Float=1.2, False Float=1)
    Loop Duration Mode=::NewEnumerator1
    UseLoopDelay=true
   )
   &SpawnBurst_Instantaneous(
    Spawn Count=&MakeCustomIntFromBool(Bool=$User.bIsFrontEndPreview, True INT=8, False INT=9)
    Use Spawn Probability=false
    UseLoopCountLimit=false
   )
   /SetVariables(Emitter.Fade:NiagaraFloat=&FN_FrontEndFloatModifier(Gameplay=0, Front End=1)) {isSetParameters=true}
   /SetVariables(Emitter.PositionFix:NiagaraFloat=&FN_FrontEndFloatModifier(Gameplay=1, Front End=1)) {isSetParameters=true}
  ]
  particleSpawnScript [
   &PartitionParticles(Number Of Partitions=&MakeCustomIntFromBool(Bool=$User.bIsFrontEndPreview, True INT=8, False INT=9))
   &InitializeParticle(
    Color=&MakeCustomLinearColorFromBool(Bool=$User.bIsFrontEndPreview, True Color=<rgba 1 1 1 1.6>, False Color=<rgba 1.5 1.5 1.5 2>)
    Mesh Scale=&MakeCustomVectorFromBool(
     Bool=$User.bIsFrontEndPreview
     True Vector=&SelectVectorFromArray(
      Direct Array Index=$Particles.PartitionParticles.Partition
      Vector Selection Array=DataInterface() {propertyValues {FloatData [<0.8 0.8 0.55>, <0.8 0.8 0.55>, <0.8 0.8 0.6>, <-0.8 0.8 0.45>, <-0.8 0.8 0.7>, <-0.8 0.8 0.72>, <1 1 0.6>, <-1 1 0.6>, <0.9 0.9 0.8>]}}
     )
     False Vector=&SelectVectorFromArray(
      Direct Array Index=$Particles.PartitionParticles.Partition
      Vector Selection Array=DataInterface() {propertyValues {FloatData [<0.8 0.8 0.55>, <0.8 0.8 0.6>, <0.8 0.8 0.66>, <-0.8 0.8 0.5>, <-0.8 0.8 0.75>, <-0.8 0.8 0.8>, <1 1 0.8>, <-1 1 0.9>, <0.9 0.9 0.8>]}}
     )
    )
    Mesh Scale Mode=::NewEnumerator4
    UsePositionOffset=false
    Write Material Random=true
   )
   &InitialMeshOrientation(
    Mesh Orientation Mode=::NewEnumerator2
    Rotation=&MakeCustomVectorFromBool(Bool=$User.bIsFrontEndPreview, True Vector=<-0.25 0 0>, False Vector=<-0.25 0 0>)
    Use Rotation Vector=true
   )
   &InitialMeshOrientation(
    Mesh Orientation Mode=::NewEnumerator2
    Rotation=&RandomRangeVector(Minimum=<0 0 0>, Maximum=<0 0.8 0>)
    Use Rotation Vector=true
   ) {module=InitialMeshOrientation001}
   &InitialMeshOrientation(
    Mesh Orientation Mode=::NewEnumerator2
    Rotation=&MakeCustomVectorFromBool(
     Bool=$User.bIsFrontEndPreview
     True Vector=&SelectVectorFromArray(
      Direct Array Index=$Particles.PartitionParticles.Partition
      Vector Selection Array=DataInterface() {propertyValues {FloatData [<0.03 0 0.12>, <0.02 0 0.1>, <0.033 0 0.1>, <0.06 0 -0.01>, <0.04 0 -0.04>, <0.05 0 -0.06>, <0.03 0 0.02>, <0 0 -0.01>, <0 0 -0.05>]}}
     )
     False Vector=&SelectVectorFromArray(
      Direct Array Index=$Particles.PartitionParticles.Partition
      Vector Selection Array=DataInterface() {propertyValues {FloatData [<0.03 0 0.12>, <0.02 0 0.1>, <0.033 0 0.1>, <0.01 0 -0.12>, <0.03 0 -0.1>, <0.035 0 -0.08>, <0.03 0 0.02>, <0 0 -0.01>, <0.02 0 -0.05>]}}
     )
    )
    Use Rotation Vector=true
   ) {module=InitialMeshOrientation002}
   &CameraOffset(Camera Offset Amount=10)
  ]
  particleUpdateScript [
   &ParticleState(Kill Particles When Lifetime Has Elapsed=false, Let Infinitely Lived Particles Die When Emitter Deactivates=true)
   &ScaleMeshSize(Scale Factor=&MakeCustomVectorFromBool(
     Bool=$User.bIsFrontEndPreview
     True Vector=<1.3 1.3 1.3>
     False Vector=&Multiply_Vector(A=<1.4 1.4 1.4>, B=&Lerp_Vector(A=<1 1 1>, B=<1 1 1.1>, Alpha=$System.Timeline.Time))
    ))
   &SkeletalMeshLocation(
    MaskSpawnGroup=false
    Mesh Sampling Type=::NewEnumerator5
    Sampled Position Offset=&AddVectorToPosition(
     Position=&MakeCustomVectorFromBool(
      Bool=$User.bIsFrontEndPreview
      True Vector=&SelectVectorFromArray(
       Direct Array Index=$Particles.PartitionParticles.Partition
       Vector Selection Array=DataInterface() {propertyValues {FloatData [<10 -75 25>, <-10 -65 10>, <-10 -55 -10>, <15 -75 25>, <16 -65 5>, <15 -55 -20>, <-20 60 0>, <30 60 0>]}}
      )
      False Vector=&SelectVectorFromArray(
       Direct Array Index=$Particles.PartitionParticles.Partition
       Vector Selection Array=DataInterface() {propertyValues {FloatData [<-10 -90 20>, <-15 -90 0>, <-20 -80 -30>, <10 -90 20>, <15 -90 5>, <20 -80 -30>, <-10 60 -50>, <20 60 -30>, <10 -90 20>]}}
      )
     )
     Vector=&Lerp_Vector(
      A=&MakeCustomVectorFromBool(Bool=$User.bIsFrontEndPreview, True Vector=<0 0 0>, False Vector=<0 0 20>)
      B=&MakeCustomVectorFromBool(Bool=$User.bIsFrontEndPreview, True Vector=<0 0 0>, False Vector=<0 0 10>)
      Alpha=&FN_PickFloatDependOnPlatform(TargetPlatformActive=1, TargetPlatformNoActive=0)
     )
    )
    Skeletal Mesh=$System.Skeletal Mesh
    Socket Sampling Mode=::NewEnumerator5
   )
   &DynamicMaterialParameters(
    Index 0 Param 1=&RandomRangeFloat(Minimum=0.5, Maximum=1)
    Index 0 Param 2=&RandomRangeFloat(Minimum=0.5, Maximum=1)
    Index 0 Param 3=&MakeCustomFloatFromBool(
     Bool=$User.bIsFrontEndPreview
     True Float=&RandomRangeFloat(Minimum=3, Maximum=5)
     False Float=&RandomRangeFloat(Minimum=10, Maximum=12)
    )
    Index 0 Param 4=1
    Index 2 Param 1=&MakeFloatFromVector(
     Vector=&AddVectorToPosition(
      Position=$Particles.Position
      Vector=&TransformVector(
       Vector=&FN_FrontEndVector3Modifier(Gameplay=$System.IGPOFix, Front End=$System.FEPoFix)
       Destination Space=::Simulation
      )
     )
     Channel=::NewEnumerator0
    )
    Index 2 Param 2=&MakeFloatFromVector(
     Vector=&AddVectorToPosition(
      Position=$Particles.Position
      Vector=&TransformVector(
       Vector=&FN_FrontEndVector3Modifier(Gameplay=$System.IGPOFix, Front End=$System.FEPoFix)
       Destination Space=::Simulation
      )
     )
     Channel=::NewEnumerator1
    )
    Index 2 Param 3=&MakeFloatFromVector(
     Vector=&AddVectorToPosition(
      Position=$Particles.Position
      Vector=&TransformVector(
       Vector=&FN_FrontEndVector3Modifier(Gameplay=$System.IGPOFix, Front End=$System.FEPoFix)
       Destination Space=::Simulation
      )
     )
     Channel=::NewEnumerator2
    )
    Index 3 Param 1=&FloatFromCurve(FloatCurve=DataInterface() {propertyValues {Curve {keys [
         <at 0 -625.0001 RCIM_Cubic>
         {arriveTangent=712.5001, interpMode=RCIM_Cubic, leaveTangent=712.5001, time=1, value=0}
         {arriveTangent=1495.8616, interpMode=RCIM_Cubic, leaveTangent=1495.8616, tangentMode=RCTM_User, time=2, value=800}
        ]}}})
    Module.Param0WriteEnabled=true
    Module.Param0WriteEnabled2=true
    Module.Param0WriteEnabled3=true
    Module.Param1WriteEnabled=true
    Module.Param1WriteEnabled2=true
    Module.Param1WriteEnabled3=true
    Module.Param2WriteEnabled=true
    Module.Param2WriteEnabled2=true
    Module.Param2WriteEnabled3=true
    Module.Param3WriteEnabled=true
    Module.Param3WriteEnabled2=false
    Module.Param3WriteEnabled3=true
    Write Parameter Index 2=true
    Write Parameter Index 3=true
   ) {module=DynamicMaterialParameters001}
   &InterpolateOverTime(
    Rate Of Change:NiagaraFloat=&FN_FrontEndFloatModifier(Gameplay=10, Front End=10)
    Target Vector Value:Vector3f=&AddVectorToPosition(
     Position=$Particles.Position
     Vector=&TransformVector(
      Vector=&FN_FrontEndVector3Modifier(Gameplay=$System.IGPOFix, Front End=$System.FEPoFix)
      Destination Space=::Simulation
     )
    )
    Variable Type=::NewEnumerator2
   )
   /SetVariables(Particles.InterpolateDynamic:Vector4f=&MakeFloat4FromVectorAndFloat(
     "Vector (RGB)"=$StackContext.InterpolateOverTime.InterpolateOverTime_V3_.Moving Average
     "Float (Alpha)"=&FN_FrontEndFloatModifier(Gameplay=1, Front End=3)
    )) {isSetParameters=true}
   &SolveForcesAndVelocity(Clamp Velocity=false, Limit Acceleration=false)
   &ScaleColor(
    Scale Alpha=&FloatFromCurve(
     FloatCurve=DataInterface() {propertyValues {Curve {keys [<at 0 0>,<at 0.25 1>]}}}
     Scale Curve=&Multiply_Float(
      A=&SelectFloatFromArray(
       Direct Array Index=$Particles.PartitionParticles.Partition
       Float Selection Array=DataInterface() {propertyValues {FloatData [1,1,1,1,1,1,0.5,0.4,1]}}
      )
      B=&Multiply_Float(
       A=&Lerp_Float(A=&FN_FrontEndFloatModifier(Gameplay=0.5, Front End=1.2), B=0.5, Alpha=$System.Timeline.Time)
       B=&FN_PickFloatDependOnPlatform(TargetPlatformActive=1, TargetPlatformNoActive=0.5)
      )
     )
    )
    Scale RGB=&MakeVectorFromLinearColorRGB(Color=&MultiplyLinearColor(
      Linear Color=$System.TODColor
      Scale Factor=&MakeVector4_FromLinearColor(LinearColor=&LerpLinearColors(
        StartColor=&MakeCustomLinearColorFromBool(Bool=$User.bIsFrontEndPreview, True Color=<rgba 1.5 0.27 0.225 1>, False Color=<rgba 1 1 1 1>)
        EndColor=<rgba 4 0.499 0.32 1>
        LerpFactor=$System.Timeline.Time
       ))
     ))
    ScaleA=true
    ScaleRGB=true
   )
  ]
  rendererClasses [&NiagaraMeshRendererProperties]
  renderers [{rendererClass=&NiagaraMeshRendererProperties, rendererIndex=0}]
 }
 SpeedLine {
  emitterUpdateScript [
   &EmitterState(
    Delay First Loop Only=true
    Inactive Response=::NewEnumerator1
    Life Cycle Mode=::NewEnumerator1
    Loop Delay=1
    Loop Duration=3
    UseLoopDelay=true
   )
   &SpawnRate(
    SpawnRate=&FN_FrontEndFloatModifier(
     Gameplay=&Multiply_Float(
      A=&Multiply_Float(A=150, B=1)
      B=&Lerp_Float(A=-0.8, B=1, Alpha=$System.Timeline.Time)
     )
     Front End=0
    )
    Use Spawn Probability=false
   )
  ]
  particleSpawnScript [
   &InitializeParticle(
    Color=<rgba 2 2 2 1>
    Lifetime Max=0.15
    Lifetime Min=0.1
    Lifetime Mode=::NewEnumerator1
    Position Offset=&FN_FrontEndVector3Modifier(Gameplay=<0 -50 0>, Front End=<0 170 0>)
    Sprite Size Max=<2 7>
    Sprite Size Min=<1.5 2>
    Sprite Size Mode=::NewEnumerator2
    UsePositionOffset=true
    Write Material Random=true
   )
   &TorusLocation(
    Align Torus With Axis=true
    Handle Radius=0
    Large Radius:NiagaraFloat=&RandomRangeFloat(
     Minimum=&FN_FrontEndFloatModifier(Gameplay=220, Front End=150)
     Maximum=&FN_FrontEndFloatModifier(Gameplay=260, Front End=150)
    )
    MaskSpawnGroup=false
    Module.Override Randomness=false
    Offset=&FN_FrontEndVector3Modifier(Gameplay=<0 100 0>, Front End=<0 -100 0>)
    Torus Axis=<0 1 0>
   )
   &AddVelocity(Velocity=<0 -500 0>)
   &AddVelocityFromPoint(
    Module.Use Falloff Distance=false
    Origin Offset=&FN_FrontEndVector3Modifier(Gameplay=<0 500 0>, Front End=<0 0 0>)
    Velocity Strength=&FN_FrontEndFloatModifier(Gameplay=1500, Front End=300)
   )
   &CameraOffset(Camera Offset Amount=-10)
  ]
  particleUpdateScript [
   &ParticleState()
   &SolveForcesAndVelocity(Clamp Velocity=false, Limit Acceleration=false)
   &ScaleSpriteSizeBySpeed(Max Scale Factor=<1 50>, Min Scale Factor=<0.01 0.05>, Velocity Threshold=1)
   &ScaleColor(
    Scale Alpha=&Multiply_Float(
     A=&FloatFromCurve(FloatCurve=DataInterface() {propertyValues {Curve {keys [
          {interpMode=RCIM_Cubic, tangentMode=RCTM_User, time=0, value=0}
          {interpMode=RCIM_Cubic, tangentMode=RCTM_User, time=0.5, value=1}
          {tangentMode=RCTM_User, time=1, value=0}
         ]}}})
     B=&Multiply_Float(
      A=&RandomRangeFloat(Minimum=0.1, Maximum=0.3)
      B=&FN_FrontEndFloatModifier(Gameplay=1, Front End=0.3)
     )
    )
    ScaleA=true
    ScaleRGB=true
   )
  ]
  rendererClasses [&NiagaraSpriteRendererProperties]
  renderers [{rendererClass=&NiagaraSpriteRendererProperties}]
 }
 Spiral_Trail {
  emitterUpdateScript [
   &EmitterState(
    Delay First Loop Only=true
    Inactive Response=::NewEnumerator1
    Life Cycle Mode=::NewEnumerator1
    Loop Delay=&MakeCustomFloatFromBool(Bool=$User.bIsFrontEndPreview, True Float=2.2, False Float=2)
    Loop Duration=4
    UseLoopDelay=true
   )
   &SpawnRate(
    SpawnRate=&MakeCustomFloatFromBool(Bool=$User.bIsFrontEndPreview, True Float=0, False Float=&Lerp_Float(A=25, B=25, Alpha=$System.Timeline.Time))
    Use Spawn Probability=true
   )
   /SetVariables(Emitter.FEPositionFix:NiagaraFloat=&FN_FrontEndFloatModifier(Gameplay=1, Front End=1)) {isSetParameters=true, module=SetVariables_#5}
   /SetVariables(Emitter.AngleFade:NiagaraFloat=&FN_FrontEndFloatModifier(Gameplay=0, Front End=2)) {isSetParameters=true, module=SetVariables_#6}
  ]
  particleSpawnScript [
   &InitializeParticle(
    Color Maximum=<rgba 0.5 0.5 0.5 1.5>
    Color Minimum=<rgba 0.15 0.15 0.15 1>
    Color Mode=::NewEnumerator2
    Lifetime Max=1.2
    Lifetime Min=0.8
    Lifetime Mode=::NewEnumerator1
    Mesh Scale Max=<2.8 5 5>
    Mesh Scale Min=<2.4 4 4>
    Mesh Scale Mode=::NewEnumerator2
    UsePositionOffset=false
    Write Material Random=true
   )
   &CameraOffset(Camera Offset Amount=40)
   &InitialMeshOrientation(
    Mesh Orientation Mode=::NewEnumerator2
    Rotation=&MakeCustomVectorFromBool(Bool=$User.bIsFrontEndPreview, True Vector=<0 0.5 0>, False Vector=<0 0 0>)
    Use Rotation Vector=true
   )
   &InitialMeshOrientation(
    Mesh Orientation Mode=::NewEnumerator2
    Rotation=&RandomRangeVector(Minimum=<-1 0 0>, Maximum=<1 0 0>)
    Use Rotation Vector=true
   ) {module=InitialMeshOrientation001}
   &AddVelocity(Velocity=<0 10 0>)
   &InitialMeshOrientation(
    Mesh Orientation Mode=::NewEnumerator2
    Rotation=&RandomRangeVector(Minimum=<0 0 -0.03>, Maximum=<0 0 0.03>)
    Use Rotation Vector=true
   ) {module=InitialMeshOrientation002}
  ]
  particleUpdateScript [
   &ParticleState()
   &SkeletalMeshLocation(
    Bone Sampling Mode=::NewEnumerator4
    MaskSpawnGroup=false
    Sampled Position Offset=&MakeCustomVectorFromBool(
     Bool=$User.bIsFrontEndPreview
     True Vector=&AddVectorToPosition(
      Position=&RandomRangeVector(Minimum=<-100 -120 50>, Maximum=<-100 -100 60>)
      Vector=&MakeVector(
       X=0
       Y=&FloatFromCurve(FloatCurve=DataInterface() {propertyValues {Curve {keys [<at 0 0>,<at 1 -40>]}}})
       Z=0
      )
     )
     False Vector=&AddVectorToPosition(
      Position=&RandomRangeVector(Minimum=<-20 190 0>, Maximum=<20 190 0>)
      Vector=&MakeVector(
       X=0
       Y=&FloatFromCurve(FloatCurve=DataInterface() {propertyValues {Curve {keys [<at 0 0>,<at 1 -40>]}}})
       Z=0
      )
     )
    )
    Skeletal Mesh=$System.Skeletal Mesh
   )
   &Drag(Drag=10, Use Linear Drag=true, Use Rotational Drag=true)
   &SolveForcesAndVelocity(Clamp Velocity=false, Limit Acceleration=false)
   &UpdateMeshOrientation(Rotation Rate=&RandomRangeFloat(Minimum=-1.5, Maximum=1))
   &ScaleMeshSize(Scale Factor=&Lerp_Vector(A=<1 1 1>, B=<1.6 1 1>, Alpha=$System.Timeline.Time)) {module=ScaleMeshSize001}
   &ScaleMeshSize(Scale Factor=&MakeVector(
     X=&FloatFromCurve(FloatCurve=DataInterface() {propertyValues {Curve {keys [<at 0 0.6>,<at 1 1>]}}})
     Y=&MakeCustomFloatFromBool(Bool=&RandomBool(), True Float=-1, False Float=1)
     Z=&MakeCustomFloatFromBool(Bool=&RandomBool(), True Float=-1, False Float=1)
    ))
   &ScaleColor(
    Scale Alpha=0.8
    Scale RGB=&MakeVectorFromLinearColorRGB(Color=$System.TODColor)
    ScaleA=true
    ScaleRGB=true
   ) {module=ScaleColor001}
   &ScaleColor(
    Scale Alpha=&FloatFromCurve(
     FloatCurve=DataInterface() {propertyValues {Curve {keys [
         {arriveTangent=0.0846, interpMode=RCIM_Cubic, leaveTangent=0.0846, tangentMode=RCTM_User, time=0, value=0}
         {arriveTangent=-0.0403, interpMode=RCIM_Cubic, leaveTangent=-0.0403, tangentMode=RCTM_User, time=0.4, value=1}
         {tangentMode=RCTM_User, time=1, value=0}
        ]}}}
     Scale Curve=&FN_FrontEndFloatModifier(Gameplay=&Lerp_Float(A=1, B=0.6, Alpha=$System.Timeline.Time), Front End=0.6)
    )
    Scale RGB=&MakeCustomVectorFromBool(
     Bool=$User.bIsFrontEndPreview
     True Vector=&Multiply_VectorByFloat(
      Vector=&MakeVectorFromLinearColorRGB(Color=&ColorFromCurve(ColorCurve=DataInterface() {propertyValues {
          AlphaCurve {keys []}
          BlueCurve {keys [<at 0.2767 0.1 RCIM_Cubic>,<at 1 1 RCIM_Cubic>]}
          GreenCurve {keys [<at 0.2767 0.1 RCIM_Cubic>,<at 1 0.8187 RCIM_Cubic>]}
          RedCurve {keys [<at 0.2767 0.1 RCIM_Cubic>,<at 1 0.6771 RCIM_Cubic>]}
         }}))
      Float=&RandomRangeFloat(Minimum=0.5, Maximum=1)
     )
     False Vector=&Multiply_VectorByFloat(
      Vector=&MakeVectorFromLinearColorRGB(Color=&FNTimeOfDayColor(
        Night Color=&ColorFromCurve(ColorCurve=DataInterface() {propertyValues {
           AlphaCurve {keys []}
           BlueCurve {keys [<at 0.284 0.2 RCIM_Cubic>,<at 0.9663 0.5 RCIM_Cubic>]}
           GreenCurve {keys [<at 0.284 0.2 RCIM_Cubic>,<at 0.9663 0.5 RCIM_Cubic>]}
           RedCurve {keys [<at 0.284 0.2 RCIM_Cubic>,<at 0.9663 0.5 RCIM_Cubic>]}
          }})
        Morning Color=&ColorFromCurve(ColorCurve=DataInterface() {propertyValues {
           AlphaCurve {keys []}
           BlueCurve {keys [<at 0.289 0.3 RCIM_Cubic>,<at 0.8487 0.2 RCIM_Cubic>]}
           GreenCurve {keys [<at 0.289 0.3 RCIM_Cubic>,<at 0.8487 0.2 RCIM_Cubic>]}
           RedCurve {keys [<at 0.289 0.3 RCIM_Cubic>,<at 0.8487 0.2 RCIM_Cubic>]}
          }})
        Day Color=&ColorFromCurve(ColorCurve=DataInterface() {propertyValues {
           AlphaCurve {keys []}
           BlueCurve {keys [<at 0.289 0.5 RCIM_Cubic>,<at 0.8487 0.2 RCIM_Cubic>]}
           GreenCurve {keys [<at 0.289 0.5 RCIM_Cubic>,<at 0.8487 0.2 RCIM_Cubic>]}
           RedCurve {keys [<at 0.289 0.5 RCIM_Cubic>,<at 0.8487 0.2 RCIM_Cubic>]}
          }})
        Evening Color=&ColorFromCurve(ColorCurve=DataInterface() {propertyValues {
           AlphaCurve {keys []}
           BlueCurve {keys [<at 0.289 0.3 RCIM_Cubic>,<at 0.8487 0.1 RCIM_Cubic>]}
           GreenCurve {keys [<at 0.289 0.3 RCIM_Cubic>,<at 0.8487 0.1 RCIM_Cubic>]}
           RedCurve {keys [<at 0.289 0.3 RCIM_Cubic>,<at 0.8487 0.1 RCIM_Cubic>]}
          }})
       ))
      Float=&RandomRangeFloat(Minimum=0.5, Maximum=1)
     )
    )
    ScaleA=true
    ScaleRGB=true
   )
   &DynamicMaterialParameters(
    Index 0 Param 1=&FloatFromCurve(FloatCurve=DataInterface() {propertyValues {Curve {keys [<at 0 1>,<at 1 -1>]}}})
    Index 0 Param 3=1
    Index 0 Param 4=&FloatFromCurve(FloatCurve=DataInterface() {propertyValues {Curve {keys [<at 0 0>,<at 1 1>]}}})
    Index 2 Param 1=&MakeFloatFromVector(
     Vector=&AddVectorToPosition(
      Position=$Particles.Position
      Vector=&TransformVector(
       Vector=&FN_FrontEndVector3Modifier(Gameplay=$System.IGCenterposition, Front End=$System.FECenterposition)
       Destination Space=::Simulation
      )
     )
     Channel=::NewEnumerator0
    )
    Index 2 Param 2=&MakeFloatFromVector(
     Vector=&AddVectorToPosition(
      Position=$Particles.Position
      Vector=&TransformVector(
       Vector=&FN_FrontEndVector3Modifier(Gameplay=$System.IGCenterposition, Front End=$System.FECenterposition)
       Destination Space=::Simulation
      )
     )
     Channel=::NewEnumerator1
    )
    Index 2 Param 3=&MakeFloatFromVector(
     Vector=&AddVectorToPosition(
      Position=$Particles.Position
      Vector=&TransformVector(
       Vector=&FN_FrontEndVector3Modifier(Gameplay=$System.IGCenterposition, Front End=$System.FECenterposition)
       Destination Space=::Simulation
      )
     )
     Channel=::NewEnumerator2
    )
    Index 3 Param 1=600
    Module.Index0Param0Write=true
    Module.Index0Param1Write=true
    Module.Index0Param2Write=true
    Module.Index0Param3Write=true
    Module.Index2Param0Write=true
    Module.Index2Param1Write=true
    Module.Index2Param2Write=true
    Module.Index2Param3Write=true
    Module.Index3Param0Write=true
    Module.Index3Param1Write=true
    Module.Index3Param2Write=true
    Module.Index3Param3Write=true
    Write Parameter Index 2=true
    Write Parameter Index 3=true
   )
   &InterpolateOverTime(
    Rate Of Change:NiagaraFloat=&FN_FrontEndFloatModifier(
     Gameplay=&Multiply_Float(A=5, B=$System.InterpolateOverTime.InterpolateOverTime_Float.Moving Average)
     Front End=8
    )
    Target Vector Value:Vector3f=&AddVectorToPosition(
     Position=$Particles.Position
     Vector=&TransformVector(
      Vector=&FN_FrontEndVector3Modifier(Gameplay=$System.IGCenterposition, Front End=$System.FECenterposition)
      Destination Space=::Simulation
     )
    )
    Variable Type=::NewEnumerator2
   )
   /SetVariables(Particles.Interpolate Dynamic=&MakeFloat4FromVectorAndFloat(
     "Vector (RGB)"=$StackContext.InterpolateOverTime.InterpolateOverTime_V3_.Moving Average
     "Float (Alpha)"=&FN_FrontEndFloatModifier(Gameplay=0, Front End=0)
    )) {isSetParameters=true, module=SetVariables_#7}
  ]
  rendererClasses [&NiagaraMeshRendererProperties]
  renderers [{rendererClass=&NiagaraMeshRendererProperties, rendererIndex=0}]
 }
 Spiral_TrailRed {
  emitterUpdateScript [
   &EmitterState(
    Delay First Loop Only=true
    Inactive Response=::NewEnumerator1
    Life Cycle Mode=::NewEnumerator1
    Loop Delay=&MakeCustomFloatFromBool(Bool=$User.bIsFrontEndPreview, True Float=2.2, False Float=2)
    Loop Duration=4
    UseLoopDelay=true
   )
   &SpawnRate(
    SpawnRate=&MakeCustomFloatFromBool(Bool=$User.bIsFrontEndPreview, True Float=0, False Float=&Lerp_Float(A=0, B=15, Alpha=$System.Timeline.Time))
    Use Spawn Probability=true
   )
   /SetVariables(Emitter.FEPositionFix:NiagaraFloat=&FN_FrontEndFloatModifier(Gameplay=1, Front End=1)) {isSetParameters=true, module=SetVariables_#5}
   /SetVariables(Emitter.AngleFade:NiagaraFloat=&FN_FrontEndFloatModifier(Gameplay=0, Front End=2)) {isSetParameters=true, module=SetVariables_#6}
  ]
  particleSpawnScript [
   &InitializeParticle(
    Lifetime Max=1.2
    Lifetime Min=0.8
    Lifetime Mode=::NewEnumerator1
    Mesh Scale Max=<2.5 1.2 1.2>
    Mesh Scale Min=<2 1 1>
    Mesh Scale Mode=::NewEnumerator2
    UsePositionOffset=false
    Write Material Random=true
   )
   &CameraOffset(Camera Offset Amount=&FN_FrontEndFloatModifier(Gameplay=15, Front End=60))
   &InitialMeshOrientation(
    Mesh Orientation Mode=::NewEnumerator2
    Rotation=&MakeCustomVectorFromBool(Bool=$User.bIsFrontEndPreview, True Vector=<0 0.5 0>, False Vector=<0 0 0>)
    Use Rotation Vector=true
   )
   &InitialMeshOrientation(
    Mesh Orientation Mode=::NewEnumerator2
    Rotation=&RandomRangeVector(Minimum=<-1 0 0>, Maximum=<1 0 0>)
    Use Rotation Vector=true
   ) {module=InitialMeshOrientation001}
   &InitialMeshOrientation(
    Mesh Orientation Mode=::NewEnumerator2
    Rotation=&MakeCustomVectorFromBool(
     Bool=$User.bIsFrontEndPreview
     True Vector=&RandomRangeVector(Minimum=<0 -0.05 0.07>, Maximum=<0 0 -0.07>)
     False Vector=&RandomRangeVector(Minimum=<0 0 0.07>, Maximum=<0 0 -0.07>)
    )
    Use Rotation Vector=true
   ) {module=InitialMeshOrientation002}
   &AddVelocity(Velocity=<0 10 0>)
  ]
  particleUpdateScript [
   &ParticleState()
   &SkeletalMeshLocation(
    Bone Sampling Mode=::NewEnumerator4
    MaskSpawnGroup=false
    Sampled Position Offset=&MakeCustomVectorFromBool(
     Bool=$User.bIsFrontEndPreview
     True Vector=&AddVectorToPosition(
      Position=&RandomRangeVector(Minimum=<-120 -150 -20>, Maximum=<-80 -170 70>)
      Vector=&MakeVector(
       X=0
       Y=&FloatFromCurve(FloatCurve=DataInterface() {propertyValues {Curve {keys [<at 0 0>,<at 1 -30>]}}})
       Z=0
      )
     )
     False Vector=&AddVectorToPosition(
      Position=&RandomRangeVector(Minimum=<-15 200 30>, Maximum=<15 200 30>)
      Vector=&MakeVector(
       X=0
       Y=&FloatFromCurve(FloatCurve=DataInterface() {propertyValues {Curve {keys [<at 0 0>,<at 1 -40>]}}})
       Z=0
      )
     )
    )
    Skeletal Mesh=$System.Skeletal Mesh
   )
   &Drag(Drag=10, Use Linear Drag=true, Use Rotational Drag=true)
   &SolveForcesAndVelocity(Clamp Velocity=false, Limit Acceleration=false)
   &UpdateMeshOrientation(Rotation Rate=&RandomRangeFloat(Minimum=-1.5, Maximum=1))
   &ScaleMeshSize(Scale Factor=<1.3 1.2 1.2>) {module=ScaleMeshSize001}
   &ScaleMeshSize(Scale Factor=&MakeVector(
     X=&FloatFromCurve(FloatCurve=DataInterface() {propertyValues {Curve {keys [<at 0 0.6>,<at 1 1>]}}})
     Y=&MakeCustomFloatFromBool(Bool=&RandomBool(), True Float=-1, False Float=1)
     Z=&MakeCustomFloatFromBool(Bool=&RandomBool(), True Float=-1, False Float=1)
    ))
   &ScaleColor(
    Scale Alpha=&FloatFromCurve(
     FloatCurve=DataInterface() {propertyValues {Curve {keys [
         {interpMode=RCIM_Cubic, tangentMode=RCTM_User, time=0, value=0}
         {arriveTangent=-0.0403, interpMode=RCIM_Cubic, leaveTangent=-0.0403, tangentMode=RCTM_User, time=0.5, value=1}
         {tangentMode=RCTM_User, time=1, value=0}
        ]}}}
     Scale Curve=&FN_FrontEndFloatModifier(Gameplay=1, Front End=0.6)
    )
    Scale RGB=&MakeVectorFromLinearColorRGB(Color=&ColorFromCurve(ColorCurve=DataInterface() {propertyValues {
        AlphaCurve {keys []}
        BlueCurve {keys [<at 0 0 RCIM_Cubic>,<at 0.2403 0.0332>,<at 0.7073 0.0005 RCIM_Cubic>]}
        GreenCurve {keys [<at 0 0 RCIM_Cubic>,<at 0.2403 0.09>,<at 0.7073 0.0005 RCIM_Cubic>]}
        RedCurve {keys [<at 0 0.1146 RCIM_Cubic>,<at 0.2403 1.5>,<at 0.7073 0.0521 RCIM_Cubic>]}
       }}))
    ScaleA=true
    ScaleRGB=true
   )
   &DynamicMaterialParameters(
    Index 0 Param 1=&FloatFromCurve(FloatCurve=DataInterface() {propertyValues {Curve {keys [<at 0 1>,<at 1 -1>]}}})
    Index 0 Param 3=1
    Index 0 Param 4=&FloatFromCurve(FloatCurve=DataInterface() {propertyValues {Curve {keys [<at 0 0>,<at 1 1>]}}})
    Index 2 Param 1=&MakeFloatFromVector(
     Vector=&AddVectorToPosition(
      Position=$Particles.Position
      Vector=&TransformVector(
       Vector=&FN_FrontEndVector3Modifier(Gameplay=$System.IGCenterposition, Front End=$System.FECenterposition)
       Destination Space=::Simulation
      )
     )
     Channel=::NewEnumerator0
    )
    Index 2 Param 2=&MakeFloatFromVector(
     Vector=&AddVectorToPosition(
      Position=$Particles.Position
      Vector=&TransformVector(
       Vector=&FN_FrontEndVector3Modifier(Gameplay=$System.IGCenterposition, Front End=$System.FECenterposition)
       Destination Space=::Simulation
      )
     )
     Channel=::NewEnumerator1
    )
    Index 2 Param 3=&MakeFloatFromVector(
     Vector=&AddVectorToPosition(
      Position=$Particles.Position
      Vector=&TransformVector(
       Vector=&FN_FrontEndVector3Modifier(Gameplay=$System.IGCenterposition, Front End=$System.FECenterposition)
       Destination Space=::Simulation
      )
     )
     Channel=::NewEnumerator2
    )
    Index 3 Param 1=600
    Module.Index0Param0Write=true
    Module.Index0Param1Write=true
    Module.Index0Param2Write=true
    Module.Index0Param3Write=true
    Module.Index2Param0Write=true
    Module.Index2Param1Write=true
    Module.Index2Param2Write=true
    Module.Index2Param3Write=true
    Module.Index3Param0Write=true
    Module.Index3Param1Write=true
    Module.Index3Param2Write=true
    Module.Index3Param3Write=true
    Write Parameter Index 2=true
    Write Parameter Index 3=true
   )
   &InterpolateOverTime(
    Rate Of Change:NiagaraFloat=&FN_FrontEndFloatModifier(
     Gameplay=&Multiply_Float(A=5, B=$System.InterpolateOverTime.InterpolateOverTime_Float.Moving Average)
     Front End=8
    )
    Target Vector Value:Vector3f=&AddVectorToPosition(
     Position=$Particles.Position
     Vector=&TransformVector(
      Vector=&FN_FrontEndVector3Modifier(Gameplay=$System.IGCenterposition, Front End=$System.FECenterposition)
      Destination Space=::Simulation
     )
    )
    Variable Type=::NewEnumerator2
   )
   /SetVariables(Particles.Interpolate Dynamic=&MakeFloat4FromVectorAndFloat(
     "Vector (RGB)"=$StackContext.InterpolateOverTime.InterpolateOverTime_V3_.Moving Average
     "Float (Alpha)"=&FN_FrontEndFloatModifier(Gameplay=0, Front End=0)
    )) {isSetParameters=true, module=SetVariables_#7}
  ]
  rendererClasses [&NiagaraMeshRendererProperties]
  renderers [{rendererClass=&NiagaraMeshRendererProperties, rendererIndex=0}]
 }
 Spiral_TrailRed001 {
  emitterUpdateScript [
   &EmitterState(
    Delay First Loop Only=true
    Inactive Response=::NewEnumerator1
    Life Cycle Mode=::NewEnumerator1
    Loop Delay=2
    Loop Duration=4
    UseLoopDelay=true
   )
   &SpawnRate(
    SpawnRate=&MakeCustomFloatFromBool(Bool=$User.bIsFrontEndPreview, True Float=25, False Float=0)
    Use Spawn Probability=true
   )
   /SetVariables(Emitter.FEPositionFix:NiagaraFloat=&FN_FrontEndFloatModifier(Gameplay=1, Front End=1)) {isSetParameters=true, module=SetVariables_#5}
   /SetVariables(Emitter.AngleFade:NiagaraFloat=&FN_FrontEndFloatModifier(Gameplay=0, Front End=2)) {isSetParameters=true, module=SetVariables_#6}
  ]
  particleSpawnScript [
   &InitializeParticle(
    Color=&RandomRangeLinearColor(Minimum=<rgba 0 0 0 1>, Maximum=<rgba 1 1 1 1>)
    Lifetime Max=1.2
    Lifetime Min=0.8
    Lifetime Mode=::NewEnumerator1
    Mesh Scale Max=<2.5 4.5 4.5>
    Mesh Scale Min=<2 6 6>
    Mesh Scale Mode=::NewEnumerator2
    UsePositionOffset=false
    Write Material Random=true
   )
   &CameraOffset(Camera Offset Amount=50)
   &InitialMeshOrientation(Mesh Orientation Mode=::NewEnumerator2, Rotation=<0 0 -0.25>, Use Rotation Vector=true)
   &InitialMeshOrientation(
    Mesh Orientation Mode=::NewEnumerator2
    Rotation=&RandomRangeVector(Minimum=<0 1 0>, Maximum=<0 0 0>)
    Use Rotation Vector=true
   ) {module=InitialMeshOrientation001}
   &AddVelocity(Velocity=<0 10 0>) {enabled=false}
   &InitialMeshOrientation(
    Mesh Orientation Mode=::NewEnumerator2
    Rotation=&RandomRangeVector(Minimum=<0 0 -0.01>, Maximum=<0 0 0.05>)
    Use Rotation Vector=true
   ) {module=InitialMeshOrientation002}
  ]
  particleUpdateScript [
   &ParticleState()
   &SkeletalMeshLocation(
    Bone Sampling Mode=::NewEnumerator4
    MaskSpawnGroup=false
    Sampled Position Offset=&AddVectorToPosition(
     Position=&RandomRangeVector(Minimum=<-90 -150 0>, Maximum=<-90 -150 60>)
     Vector=&MakeVector(
      X=0
      Y=&FloatFromCurve(FloatCurve=DataInterface() {propertyValues {Curve {keys [<at 0 0>,<at 1 -40>]}}})
      Z=0
     )
    )
    Skeletal Mesh=$System.Skeletal Mesh
   )
   &Drag(Drag=10, Use Linear Drag=true, Use Rotational Drag=true)
   &SolveForcesAndVelocity(Clamp Velocity=false, Limit Acceleration=false)
   &UpdateMeshOrientation(Rotation Rate=&RandomRangeFloat(Minimum=-1.5, Maximum=1))
   &ScaleMeshSize(Scale Factor=&MakeVector(
     X=&FloatFromCurve(FloatCurve=DataInterface() {propertyValues {Curve {keys [<at 0 0.8>,<at 1 1.25>]}}}, Scale Curve=1.2)
     Y=&MakeCustomFloatFromBool(Bool=&RandomBool(), True Float=-1, False Float=1)
     Z=&MakeCustomFloatFromBool(Bool=&RandomBool(), True Float=-1, False Float=1)
    ))
   &ScaleColor(
    Scale Alpha=&FloatFromCurve(FloatCurve=DataInterface() {propertyValues {Curve {keys [
         {interpMode=RCIM_Cubic, tangentMode=RCTM_User, time=0, value=0}
         {arriveTangent=-0.0403, interpMode=RCIM_Cubic, leaveTangent=-0.0403, tangentMode=RCTM_User, time=0.4, value=1}
         {tangentMode=RCTM_User, time=1, value=0}
        ]}}})
    Scale RGB=&MakeVectorFromLinearColorRGB(Color=&Multiply_LinearColorByFloat(
      Linear Color=&ColorFromCurve(ColorCurve=DataInterface() {propertyValues {
         AlphaCurve {keys []}
         BlueCurve {keys [<at 0 0 RCIM_Cubic>,<at 0.2689 0.0332>,<at 0.7118 0.0005 RCIM_Cubic>]}
         GreenCurve {keys [<at 0 0 RCIM_Cubic>,<at 0.2689 0.09>,<at 0.7118 0.0005 RCIM_Cubic>]}
         RedCurve {keys [<at 0 0.1146 RCIM_Cubic>,<at 0.2689 1.5>,<at 0.7118 0.0521 RCIM_Cubic>]}
        }})
      Float=&FloatFromCurve(FloatCurve=DataInterface() {propertyValues {Curve {keys [
           {interpMode=RCIM_Cubic, tangentMode=RCTM_User, time=0, value=0}
           {arriveTangent=7.0217, interpMode=RCIM_Cubic, leaveTangent=-5.5711, tangentMode=RCTM_Break, time=0.45, value=1}
           {tangentMode=RCTM_User, time=1, value=0}
          ]}}})
     ))
    ScaleA=true
    ScaleRGB=true
   )
   &DynamicMaterialParameters(
    Index 0 Param 1=&FloatFromCurve(FloatCurve=DataInterface() {propertyValues {Curve {keys [<at 0 1>,<at 1 -1>]}}})
    Index 0 Param 3=1
    Index 0 Param 4=&FloatFromCurve(FloatCurve=DataInterface() {propertyValues {Curve {keys [{arriveTangent=0.3048, interpMode=RCIM_Cubic, leaveTangent=0.3048, tangentMode=RCTM_User, time=0, value=0}, <at 1 1 RCIM_Cubic>]}}})
    Index 2 Param 1=&MakeFloatFromVector(
     Vector=&AddVectorToPosition(
      Position=$Particles.Position
      Vector=&TransformVector(
       Vector=&FN_FrontEndVector3Modifier(Gameplay=$System.IGCenterposition, Front End=$System.FECenterposition)
       Destination Space=::Simulation
      )
     )
     Channel=::NewEnumerator0
    )
    Index 2 Param 2=&MakeFloatFromVector(
     Vector=&AddVectorToPosition(
      Position=$Particles.Position
      Vector=&TransformVector(
       Vector=&FN_FrontEndVector3Modifier(Gameplay=$System.IGCenterposition, Front End=$System.FECenterposition)
       Destination Space=::Simulation
      )
     )
     Channel=::NewEnumerator1
    )
    Index 2 Param 3=&MakeFloatFromVector(
     Vector=&AddVectorToPosition(
      Position=$Particles.Position
      Vector=&TransformVector(
       Vector=&FN_FrontEndVector3Modifier(Gameplay=$System.IGCenterposition, Front End=$System.FECenterposition)
       Destination Space=::Simulation
      )
     )
     Channel=::NewEnumerator2
    )
    Index 3 Param 1=600
    Module.Index0Param0Write=true
    Module.Index0Param1Write=true
    Module.Index0Param2Write=true
    Module.Index0Param3Write=true
    Module.Index2Param0Write=true
    Module.Index2Param1Write=true
    Module.Index2Param2Write=true
    Module.Index2Param3Write=true
    Module.Index3Param0Write=true
    Module.Index3Param1Write=true
    Module.Index3Param2Write=true
    Module.Index3Param3Write=true
    Write Parameter Index 2=true
    Write Parameter Index 3=true
   )
   &InterpolateOverTime(
    Rate Of Change:NiagaraFloat=&FN_FrontEndFloatModifier(
     Gameplay=&Multiply_Float(A=5, B=$System.InterpolateOverTime.InterpolateOverTime_Float.Moving Average)
     Front End=8
    )
    Target Vector Value:Vector3f=&AddVectorToPosition(
     Position=$Particles.Position
     Vector=&TransformVector(
      Vector=&FN_FrontEndVector3Modifier(Gameplay=$System.IGCenterposition, Front End=$System.FECenterposition)
      Destination Space=::Simulation
     )
    )
    Variable Type=::NewEnumerator2
   )
   /SetVariables(Particles.Interpolate Dynamic=&MakeFloat4FromVectorAndFloat(
     "Vector (RGB)"=$StackContext.InterpolateOverTime.InterpolateOverTime_V3_.Moving Average
     "Float (Alpha)"=&FN_FrontEndFloatModifier(Gameplay=0, Front End=0)
    )) {isSetParameters=true, module=SetVariables_#7}
  ]
  rendererClasses [&NiagaraMeshRendererProperties]
  renderers [{rendererClass=&NiagaraMeshRendererProperties, rendererIndex=0}]
 }
 mouth {
  emitterUpdateScript [
   &EmitterState(
    Inactive Response=::NewEnumerator1
    Life Cycle Mode=::NewEnumerator1
    Loop Behavior=::NewEnumerator1
    Loop Delay=0.5
    Loop Duration Mode=::NewEnumerator1
    UseLoopDelay=true
   )
   &SpawnBurst_Instantaneous(Spawn Count=2, Use Spawn Probability=false, UseLoopCountLimit=false)
  ]
  particleSpawnScript [
   &PartitionParticles()
   &InitializeParticle(
    Color=<rgba 1 1 1 2>
    Mesh Scale=&SelectVectorFromArray(
     Direct Array Index=$Particles.PartitionParticles.Partition
     Vector Selection Array=DataInterface() {propertyValues {FloatData [<8 6 12>, <-8 6 12>]}}
    )
    Mesh Scale Mode=::NewEnumerator4
    UsePositionOffset=false
    Write Material Random=true
   )
   &InitialMeshOrientation(Mesh Orientation Mode=::NewEnumerator2, Rotation=<0 0 0.5>, Use Rotation Vector=true)
   &InitialMeshOrientation(
    Mesh Orientation Mode=::NewEnumerator2
    Rotation=&SelectVectorFromArray(
     Direct Array Index=$Particles.PartitionParticles.Partition
     Vector Selection Array=DataInterface() {propertyValues {FloatData [<0.03 -0.2 0>, <0.04 -0.3 0.03>]}}
    )
    Use Rotation Vector=true
   ) {module=InitialMeshOrientation001}
  ]
  particleUpdateScript [
   &ParticleState(Kill Particles When Lifetime Has Elapsed=false, Let Infinitely Lived Particles Die When Emitter Deactivates=true)
   &SkeletalMeshLocation(
    Bone / Socket Index=2
    Bone Sampling Mode=::NewEnumerator4
    MaskSpawnGroup=false
    Sampled Position Offset=&MakeCustomVectorFromBool(
     Bool=$User.bIsFrontEndPreview
     True Vector=&AddVectorToPosition(
      Position=&SelectVectorFromArray(
       Direct Array Index=$Particles.PartitionParticles.Partition
       Vector Selection Array=DataInterface() {propertyValues {FloatData [<55 -10 10>, <55 -10 -10>]}}
      )
      Vector=&MakeVector(
       X=&FloatFromCurve(FloatCurve=DataInterface() {propertyValues {Curve {keys [<at 0 -20>,<at 1 0>]}}})
       Y=0
       Z=0
      )
     )
     False Vector=&SelectVectorFromArray(
      Direct Array Index=$Particles.PartitionParticles.Partition
      Vector Selection Array=DataInterface() {propertyValues {FloatData [<55 -10 0>, <55 -10 0>]}}
     )
    )
    Skeletal Mesh=$System.Skeletal Mesh
   )
   &ScaleMeshSize(Scale Factor=&VectorFromFloat(Value=&FloatFromCurve(FloatCurve=DataInterface() {propertyValues {Curve {keys [<at 0 0.3>,<at 0.45 1>]}}}))) {module=ScaleMeshSize001}
   &ScaleMeshSize(Scale Factor=&Lerp_Vector(A=<1 1 1>, B=<1 1 1>, Alpha=$User.ParaGlideLeanAlpha))
   /SetVariables(Particles.Color002=&MakeCustomLinearColorFromBool(Bool=$User.bIsFrontEndPreview, True Color=<rgba 0.026 0.0015 0.0015 1.5>, False Color=<rgba 0.026 0.0015 0.0015 1.5>)) {isSetParameters=true}
   /SetVariables(Particles.Color003:LinearColor=&MakeCustomLinearColorFromBool(Bool=$User.bIsFrontEndPreview, True Color=<rgba 1 0.0238 0.02 1>, False Color=<rgba 2 0.0475 0.04 1>)) {isSetParameters=true}
   &SolveForcesAndVelocity(Clamp Velocity=false, Limit Acceleration=false)
   &InterpolateOverTime(Target Vector Value:NiagaraPosition=$Particles.Position, Variable Type=::NewEnumerator2)
   /SetVariables(Particles.InterpolateDynamic:Vector4f=&MakeFloat4FromVectorAndFloat(
     "Vector (RGB)"=$StackContext.InterpolateOverTime.InterpolateOverTime_V3_.Moving Average
     "Float (Alpha)"=&FN_FrontEndFloatModifier(Gameplay=0.5, Front End=3)
    )) {isSetParameters=true}
   &DynamicMaterialParameters(
    Index 3 Param 1=&MakeCustomFloatFromBool(
     Bool=$User.bIsFrontEndPreview
     True Float=&FloatFromCurve(FloatCurve=DataInterface() {propertyValues {Curve {keys [{arriveTangent=237.3113, time=0, value=-600}, {arriveTangent=816.2405, leaveTangent=816.2405, time=0.5, value=0}]}}})
     False Float=&FloatFromCurve(
      FloatCurve=DataInterface() {propertyValues {Curve {keys [{arriveTangent=237.3113, time=0, value=-600}, {arriveTangent=816.2405, leaveTangent=816.2405, time=1, value=0}]}}}
      Scale Curve=&RandomRangeFloat(Minimum=0.8, Maximum=1)
     )
    )
    Module.Index3Float0Write=true
    Module.Index3Float1Write=true
    Module.Index3Float2Write=true
    Module.Index3Float3Write=true
    Write Parameter Index 0=::NewEnumerator0
    Write Parameter Index 3=::NewEnumerator1
   )
  ]
  rendererClasses [&NiagaraMeshRendererProperties]
  renderers [
   {rendererClass=&NiagaraMeshRendererProperties, rendererIndex=0}
   {rendererClass=&NiagaraMeshRendererProperties, rendererIndex=1}
  ]
 }
}
summary {
 systemName=NS_Glider_SherWolf_Idle
 emitters [
  {emitterName=Head_Trail, rendererClasses [&NiagaraMeshRendererProperties]}
  {emitterName=Head_Smoke, rendererClasses [&NiagaraMeshRendererProperties]}
  {emitterName=Hind_Smoke, rendererClasses [&NiagaraSpriteRendererProperties]}
  {emitterName=Hind_Smoke_Low, rendererClasses [&NiagaraSpriteRendererProperties]}
  {emitterName=Spiral_TrailRed, rendererClasses [&NiagaraMeshRendererProperties]}
  {emitterName=Spiral_TrailRed001, rendererClasses [&NiagaraMeshRendererProperties]}
  {emitterName=Spiral_Trail, rendererClasses [&NiagaraMeshRendererProperties]}
  {emitterName=Hind_Trail, rendererClasses [&NiagaraMeshRendererProperties]}
  {emitterName=SpeedLine, rendererClasses [&NiagaraSpriteRendererProperties]}
  {emitterName=mouth, rendererClasses [&NiagaraMeshRendererProperties]}
 ]
 userVariables [
  User.bIsFrontEnd {type=NiagaraBool}
  User.bIsFrontEndPreview {type=NiagaraBool}
  User.ForwardDot {type=NiagaraFloat, defaultValue=0, description=""}
  User.ParaGlideLeanAlpha {type=NiagaraFloat, defaultValue=0, description=""}
 ]
}
